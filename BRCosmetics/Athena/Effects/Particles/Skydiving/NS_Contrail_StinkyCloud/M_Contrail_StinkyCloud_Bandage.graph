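# Masked, fully rough Default Lit material for Niagara ribbons, drawing the bandage part of the StinkyCloud skydiving contrail.
# It multiplies the bandage diffuse texture's RGB by particle color.
# A separate chain offsets UVs with the green channel of a time-panned cloud texture, but the bandage sample does not use the result.
# The graph shows no links to material outputs.
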Material /BRCosmetics/Athena/Effects/Particles/Skydiving/NS_Contrail_StinkyCloud/M_Contrail_StinkyCloud_Bandage
graph {
 properties {
  bCastRayTracedShadows=false
  bFullyRough=true
  bUsedWithNiagaraRibbons=true
  blendMode=BLEND_Masked
  pixelDepthOffsetMode=PDOM_Legacy
  usedShadingModels=Default Lit
 }
 nodes [
  n1 = TextureCoordinate()
  n2 = Time()
  n3 = TextureSample(UVs=Add(A=Multiply(A=@n1) {constB=0.3}, B=Multiply(A=@n2) {constB=0.5})) {texture=/Game/Effects/Fort_Effects/Textures/T_Contrail_StinkyCloud_ColorAdditive}
  Add(A=TextureCoordinate(), B=Multiply(A=AppendVector(A=@n3.G, B=@n3.G)) {constB=0.3})
  Constant()
  Multiply(
   A=TextureSample() {texture=&/T_Contrail_StinkyCloud_Bandage_D}.RGB
   B=ParticleColor().RGB
  )
 ]
}
