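# Translucent unlit Niagara material for the ribbon of the StinkyCloud skydiving contrail cosmetic.
# Opacity comes from a scrolling, particle-randomized ribbon texture, tapered along U and eroded, then multiplied by particle alpha, camera fade and depth fade.
# Emissive adds a faint scrolling color texture to particle color.
# World position offset nudges toward the camera by default, with a parameter blending to SplineThicken widths driven by dynamic parameters.
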Material /BRCosmetics/Athena/Effects/Particles/Skydiving/NS_Contrail_StinkyCloud/M_Contrail_StinkyCloud_Ribbon
graph {
 properties {
  bCastRayTracedShadows=false
  bUsedWithNiagaraMeshParticles=true
  bUsedWithNiagaraRibbons=true
  blendMode=BLEND_Translucent
  pixelDepthOffsetMode=PDOM_Legacy
  shadingModel=MSM_Unlit
  usedShadingModels=Unlit
 }
 nodes [
  n1 = Subtract(A=ObjectPositionWS(), B=CameraPositionWS())
  n2 = ScalarParameter() {defaultValue=40, parameterName=SplineThicken_BaseWidth}
  n3 = DynamicParameter() {parameterIndex=2, paramNames [SplineThicken_WidthMult, "", "", ""]}
  n4 = ScalarParameter() {defaultValue=80, parameterName=SplineThicken_TipWidth}
  LinearInterpolate(
   A=Multiply(A=Normalize(VectorInput=@n1)) {constB=-6}
   B=MaterialFunctionCall(
    WidthBase=Multiply(A=@n2, B=@n3.SplineThicken_WidthMult)
    WidthTip=Multiply(A=@n4, B=@n3.SplineThicken_WidthMult)
    Expand U or V UV Channel=StaticBool()
   ) {materialFunction=&SplineThicken}.WorldPositionOffset
   Alpha=ScalarParameter() {parameterName=UseSplineThicken}
  )
  n5 = ParticleColor()
  n6 = Constant() {r=0.15}
  n7 = TextureCoordinate()
  n8 = Multiply(A=ParticleRandom()) {constB=5}
  n9 = Add(A=Time(), B=@n8)
  n10 = TextureSample(UVs=Add(A=Multiply(A=@n7) {constB=0.2}, B=Multiply(A=@n9) {constB=-0.1})) {texture=&/T_Contrail_StinkyCloud_ColorAdditive}
  MaterialFunctionCall(Emissive=Add(A=@n5.RGB, B=Multiply(A=@n6, B=@n10.RGB))) {materialFunction=&MF_FX_EmissiveAdjustments}
  n11 = TextureCoordinate()
  n12 = AppendVector(A=ScalarParameter() {defaultValue=1, parameterName=RibbonScale}, B=Constant() {r=1})
  n13 = Time()
  n14 = Multiply(A=Multiply(A=Add(A=@n13, B=@n8), B=ScalarParameter() {defaultValue=1, parameterName=ScrollSpeed})) {constB=-1}
  n15 = Constant()
  n16 = TextureSample(UVs=Add(A=Multiply(A=@n11, B=@n12), B=AppendVector(A=@n14, B=@n15))) {
   samplerType=SAMPLERTYPE_LinearColor
   texture=&/T_Contrail_StinkyCloud_Ribbon
  }
  n17 = ComponentMask(None=TextureCoordinate()) {r=true}
  n18 = OneMinus(None=@n17)
  n19 = Saturate(None=Subtract(A=Subtract(A=@n16.R, B=@n17), B=SmoothStep(Value=@n18) {constMin=0.93}))
  n20 = ScalarParameter() {defaultValue=1, parameterName=TextureMultiply}
  n21 = Saturate(None=MaterialFunctionCall!(Gradient=Multiply(A=@n19, B=@n20) {constB=2}, Erosion Amount=ScalarParameter() {defaultValue=0.45, parameterName=Noise Erosion, sliderMax=1}) {materialFunction=/BRCosmetics/Athena/Effects/Particles/Skydiving/NS_Contrail_DarkAndLight/Functions/MF_ErosionNormalized}.Result)
  n22 = ScalarParameter() {defaultValue=64, group=Camera Depth Fade, parameterName="Fade Length (S)"}
  n23 = ScalarParameter() {defaultValue=20, group=Camera Depth Fade, parameterName="Fade Offset (S)"}
  Multiply(
   A=Multiply(
    A=Multiply(A=@n21, B=@n5.A)
    B=MaterialFunctionCall(Fade Length=@n22, Fade Offset=@n23) {materialFunction=&CameraDepthFade}.Result
   )
   B=DepthFade(FadeDistance=ScalarParameter() {defaultValue=25, parameterName=DepthFade}) {fadeDistanceDefault=25}
  )
 ]
}
