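# Niagara skydiving contrail cosmetic that trails a stinky purple-and-green cloud from the character's body.
# Ribbon pairs, smoke, glow decals and attached sprites follow bones on User skeletal meshes, with thin curl-noise wisps and pulsing firefly mesh/sprite particles alongside.
# ParaGlideLeanAlpha blends many velocities, lifetimes, noise strengths and spawn rates, while FrontEnd modifiers substitute values for the front end.
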NiagaraSystem /BRCosmetics/Athena/Effects/Particles/Skydiving/NS_Contrail_StinkyCloud/NS_Contrail_StinkyCloud_v4
dependencies {
 usedDataInterfaces [
  &NiagaraDataInterfaceCurve
  &NiagaraDataInterfaceColorCurve
  &NiagaraDataInterfaceVectorCurve
  &NiagaraDataInterfaceVector2DCurve
  &NiagaraDataInterfaceArrayFloat
 ]
 usedDynamicInputs [&VectorFromCurve, &SetBoolByFloatComparison]
 usedModules [
  &SystemState
  /Game/Effects/Niagara/Modules/Skydiving/FN_DivingAttributes/v5/FN_DivingAttributes
  &self:SystemSpawnScript.SystemScriptSource.SystemScriptGraph.NiagaraNodeAssignment_0.SetVariables_#1
  &FN_Integrator
  &Timeline
 ]
}
emitters {
 Decal {
  emitterUpdateScript [
   &EmitterState()
   /Game/Effects/Niagara/Modules/Spawn/SpawnRateFromMultipleSpawnGroups_2(Global Spawnrate=4)
   &SpawnRate(SpawnRate=14, Use Spawn Probability=false) {enabled=false}
  ]
  particleSpawnScript [
   &FN_SetSpawnGroup()
   /Niagara/Modules/Spawn/Initialization/InitializeParticle(
    Color=<rgba 0.2941 0.9062 0.6352 0.35>
    Lifetime=0.6
    Module.Write Color=true
    Module.Write Lifetime=true
    Module.Write Mass=true
    Module.Write Material Random=true
    Module.Write Position=false
    Module.Write Scale=false
    Module.Write SpriteRotation=false
    Module.Write SpriteSize=true
    Position=&SimulationPosition()
    Sprite Size=&Vector2DFromFloat(Value=HlslExpression() {hlslExpression="Particles.SpawnGroup < 2 ? 40 : 25"})
   )
   &CameraOffset(Camera Offset Amount=9)
  ]
  particleUpdateScript [
   &ParticleState()
   &SkeletalMeshLocation(
    Bone / Socket Index=$Particles.SpawnGroup
    Bone Sampling Mode=::NewEnumerator4
    Inherit Velocity=false
    Interpolate Mesh Sampling=false
    MaskSpawnGroup=false
    Skeletal Mesh=$User.SkeletalMesh
   )
   &ScaleColor(
    Scale Alpha=&FloatFromCurve(
     FloatCurve=DataInterface() {propertyValues {Curve {keys [<at 0.0001 0.3382>,<at 0.1393 0.6251>,<at 0.3202 0.5233>,<at 0.4613 0.8934>,<at 0.6241 0.3845>,<at 0.7724 0.6066>,<at 1.0006 0.3567>]}}}
     Scale Curve=HlslExpression() {hlslExpression="Particles.SpawnGroup < 2 ? 1 : 0.5"}
    )
    ScaleA=true
    ScaleRGB=false
   )
  ]
  rendererClasses [&NiagaraSpriteRendererProperties]
  renderers [{rendererClass=&NiagaraSpriteRendererProperties}]
 }
 Empty {
  emitterUpdateScript [
   &EmitterState()
   &FN_SpawnRibbon_v2(Number of Ribbons=2, Spawn Mode=::NewEnumerator0, SpawnRate=128)
  ]
  particleSpawnScript [
   &FN_SetSpawnGroup(Number of Spawn Groups=2)
   &FN_InitializeRibbon_v3(Color:LinearColor=<rgba 0.8438 0.7611 0.6064 1>, Lifetime=0.3, Ribbon Width=1.5)
   &SkeletalMeshLocation(
    Bone / Socket Index=&Add_Integer(A=$Particles.SpawnGroup, B=2)
    Bone Sampling Mode=::NewEnumerator4
    MaskSpawnGroup=false
    Skeletal Mesh=$User.SkeletalMesh2
   )
   &AddVelocity(Velocity=&FN_FrontEndVector3Modifier(Gameplay=&Lerp_Vector(A=<0 -50 350>, B=<1 -140 700>, Alpha=$User.ParaGlideLeanAlpha), Front End=<0 -100 350>))
  ]
  particleUpdateScript [
   &CurlNoiseForce(Noise Frequency=5, Noise Strength=&Lerp_Float(A=650, B=1000, Alpha=$User.ParaGlideLeanAlpha), Pan Noise=true)
   &SolveForcesAndVelocity(Clamp Velocity=false, Limit Acceleration=false)
   &ParticleState()
   &FN_UpdateRibbon()
   &ScaleRibbonWidth(Ribbon Width Scale=&FloatFromCurve(FloatCurve=DataInterface() {propertyValues {Curve {keys [
         {interpMode=RCIM_Cubic, leaveTangent=2.855, tangentMode=RCTM_Break, time=0, value=1}
         {arriveTangent=0.2168, interpMode=RCIM_Cubic, tangentMode=RCTM_Break, time=1, value=2.375}
        ]}}}))
  ]
  rendererClasses [&NiagaraRibbonRendererProperties]
  renderers [{rendererClass=&NiagaraRibbonRendererProperties}]
 }
 Empty001 {
  emitterUpdateScript [
   &EmitterState()
   &FN_SpawnRibbon_v2(Number of Ribbons=2, Spawn Mode=::NewEnumerator0, SpawnRate=128)
  ]
  particleSpawnScript [
   &FN_SetSpawnGroup(Number of Spawn Groups=2)
   &FN_InitializeRibbon_v3(Color:LinearColor=<rgba 0.7188 0.6483 0.5166 1>, Lifetime=0.3, Ribbon Width=0.8)
   &SkeletalMeshLocation(
    Bone / Socket Index=&Add_Integer(A=$Particles.SpawnGroup, B=2)
    Bone Sampling Mode=::NewEnumerator4
    MaskSpawnGroup=false
    Skeletal Mesh=$User.SkeletalMesh2
   )
   &AddVelocity(Velocity=&FN_FrontEndVector3Modifier(Gameplay=&Lerp_Vector(A=<0 -100 200>, B=<1 -180 500>, Alpha=$User.ParaGlideLeanAlpha), Front End=<0 -120 200>))
  ]
  particleUpdateScript [
   &CurlNoiseForce(
    Noise Frequency=5
    Noise Strength=&Lerp_Float(A=750, B=1250, Alpha=$User.ParaGlideLeanAlpha)
    Pan Noise=true
    Random Seed=20
   )
   &SolveForcesAndVelocity(Clamp Velocity=false, Limit Acceleration=false)
   &ParticleState()
   &FN_UpdateRibbon()
   &ScaleRibbonWidth(Ribbon Width Scale=&FloatFromCurve(FloatCurve=DataInterface() {propertyValues {Curve {keys [
         {interpMode=RCIM_Cubic, leaveTangent=2.855, tangentMode=RCTM_Break, time=0, value=1}
         {arriveTangent=0.2168, interpMode=RCIM_Cubic, tangentMode=RCTM_Break, time=1, value=2.375}
        ]}}}))
  ]
  rendererClasses [&NiagaraRibbonRendererProperties]
  renderers [{rendererClass=&NiagaraRibbonRendererProperties}]
 }
 FN_AttachedParticle_v3 {
  emitterUpdateScript [
   &EmitterState()
   &SpawnRate(SpawnRate=&Lerp_Float(A=15, B=30, Alpha=$User.ParaGlideLeanAlpha), Use Spawn Probability=false)
  ]
  particleSpawnScript [
   &FN_SetSpawnGroup(Number of Spawn Groups=5)
   /Niagara/Modules/Spawn/Initialization/V2/InitializeParticle(
    Color Maximum=<rgba 0.2036 0.0427 0.6302 1>
    Color Minimum=<rgba 0.0453 0.6667 0.4064 1>
    Color Mode=::NewEnumerator2
    Lifetime Max=1.2
    Lifetime Min=0.5
    Lifetime Mode=::NewEnumerator1
    Sprite Rotation Mode=::NewEnumerator2
    Sprite Size Mode=::NewEnumerator1
    Uniform Sprite Size Max=12
    Uniform Sprite Size Min=8
    UsePositionOffset=false
    Write Material Random=true
   ) {module=InitializeParticle001}
   /Game/Effects/Niagara/Modules/Sprite/FN_SpriteUV(End Frame=3)
   &SkeletalMeshLocation(
    Bone / Socket Index=$Particles.SpawnGroup
    Bone Sampling Mode=::NewEnumerator4
    Inherit Velocity=false
    Interpolate Mesh Sampling=false
    MaskSpawnGroup=false
    Skeletal Mesh=$User.SkeletalMesh
   )
   &ShapeLocation(MaskSpawnGroup=false, Override Seed=false, Sphere Radius=18, UseOwnerScale=true)
   &AddVelocity(Velocity=&FN_FrontEndVector3Modifier(
     Gameplay=&Lerp_Vector(
      A=&RandomRangeVector(Minimum=<30 55 150>, Maximum=<-30 -55 350>)
      B=&RandomRangeVector(Minimum=<-40 -40 125>, Maximum=<40 40 250>)
      Alpha=$User.ParaGlideLeanAlpha
     )
     Front End=&RandomRangeVector(Minimum=<0 -40 30>, Maximum=<0 -40 85>)
    ))
  ]
  particleUpdateScript [
   &ParticleState()
   &Drag(Drag=2, Use Linear Drag=true, Use Rotational Drag=true)
   &ScaleSpriteSize(
    Non-Uniform Curve Sprite Scale=DataInterface() {propertyValues {
      XCurve {keys [
        {interpMode=RCIM_Cubic, leaveTangent=30.217, tangentMode=RCTM_Break, time=0, value=0}
        {interpMode=RCIM_Cubic, tangentMode=RCTM_Break, time=0.1, value=1}
        {interpMode=RCIM_Cubic, leaveTangent=-1.1133, tangentMode=RCTM_Break, time=0.4, value=1}
        {arriveTangent=-1.1701, interpMode=RCIM_Cubic, leaveTangent=-1.1153, tangentMode=RCTM_Break, time=0.7, value=0.6}
        {interpMode=RCIM_Cubic, tangentMode=RCTM_Break, time=1, value=0.45}
       ]}
      YCurve {keys [
        {interpMode=RCIM_Cubic, leaveTangent=19.2104, tangentMode=RCTM_Break, time=0, value=0}
        {arriveTangentWeight=0.0333, interpMode=RCIM_Cubic, leaveTangent=0.0043, leaveTangentWeight=0.1, tangentMode=RCTM_Break, time=0.1, value=1}
        {interpMode=RCIM_Cubic, leaveTangent=-3.6039, tangentMode=RCTM_Break, time=0.4, value=1}
        {interpMode=RCIM_Cubic, leaveTangent=2.2301, tangentMode=RCTM_Break, time=0.6, value=0.525}
        {arriveTangent=0.8194, interpMode=RCIM_Cubic, tangentMode=RCTM_Break, time=1, value=1.25}
       ]}
     }}
    Scale Sprite Size Mode=::NewEnumerator3
   )
   &SolveForcesAndVelocity(Clamp Velocity=false, Limit Acceleration=false)
   &ScaleColor(
    Scale Alpha=&FloatFromCurve(FloatCurve=DataInterface() {propertyValues {Curve {keys [
         {arriveTangent=5.8658, interpMode=RCIM_Cubic, leaveTangent=5.8658, tangentMode=RCTM_User, time=0, value=0.325}
         {arriveTangent=-2.5249, interpMode=RCIM_Cubic, leaveTangent=0.5198, tangentMode=RCTM_Break, time=0.1408, value=0.9847}
         {arriveTangent=-3.2289, interpMode=RCIM_Cubic, leaveTangent=-3.2289, tangentMode=RCTM_User, time=1, value=0.25}
        ]}}})
    ScaleA=true
    ScaleRGB=true
   )
  ]
  rendererClasses [&NiagaraSpriteRendererProperties]
  renderers [{rendererClass=&NiagaraSpriteRendererProperties}]
 }
 FN_Fireflies_Optimized {
  bEnabled=false
  emitterUpdateScript [
   &EmitterState(
    Enable Distance Culling=true
    Inactive Response=::NewEnumerator1
    Life Cycle Mode=::NewEnumerator1
    Loop Delay=&UniformRangedFloat(Minimum=0.5, Maximum=2.5)
    Loop Duration=&UniformRangedFloat(Minimum=0.6, Maximum=1.8)
    MaxDistance=8000
    Module.bUseMaxDistance=true
    Module.bUseMinDistance=true
    Recalculate Duration Each Loop=true
    Scalability Mode=::NewEnumerator1
    UseLoopDelay=false
   )
   &SpawnBurst_Instantaneous(Spawn Count=3, Use Spawn Probability=false, UseLoopCountLimit=false)
   &SpawnRate(SpawnRate=&Lerp_Float(A=4.5, B=8, Alpha=$User.ParaGlideLeanAlpha), Use Spawn Probability=false)
  ]
  particleSpawnScript [
   /Niagara/Modules/Spawn/Initialization/InitializeParticle(
    Lifetime=&UniformRangedFloat(Minimum=1, Maximum=2.5)
    Mesh Scale=&Multiply_VectorByFloat(Vector=<1 2 1>, Float=&UniformRangedFloat(Minimum=0.16, Maximum=0.4))
    Module.Write Color=true
    Module.Write Lifetime=true
    Module.Write Mass=true
    Module.Write Material Random=true
    Module.Write Position=false
    Module.Write Scale=true
    Module.Write SpriteRotation=false
    Module.Write SpriteSize=true
    Position=&SimulationPosition()
    Sprite Size=<30 50>
   )
   &DynamicMaterialParameters(
    Index 0 Param 1=10
    Index 0 Param 2=1
    Index 1 Param 1=&UniformRangedFloat(Minimum=0, Maximum=30)
    Index 1 Param 2=&UniformRangedFloat(Minimum=0, Maximum=1)
    Index 1 Param 3=&UniformRangedFloat(Minimum=0, Maximum=1)
    Index 1 Param 4=&UniformRangedFloat(Minimum=0, Maximum=1)
    Module.Param0WriteEnabled=true
    Module.Param0WriteEnabled1=true
    Module.Param1WriteEnabled=true
    Module.Param1WriteEnabled1=false
    Module.Param2WriteEnabled=true
    Module.Param2WriteEnabled1=false
    Module.Param3WriteEnabled=true
    Module.Param3WriteEnabled1=false
    Write Parameter Index 1=true
   )
   /Niagara/Modules/Spawn/Location/SphereLocation(Hemisphere Z=true, MaskSpawnGroup=false, Module.Override Randomness=false, Offset=<0 0 60>, Surface Only=false)
   /SetVariables(Particles.Random_Color:LinearColor=&ColorFromCurve(
     ColorCurve=DataInterface() {propertyValues {
       AlphaCurve {keys [<at 0 0.13 RCIM_Cubic>]}
       BlueCurve {keys [<at 0 0.5987 RCIM_Cubic>,<at 0.4434 0.6204>,<at 1 0.7167 RCIM_Cubic>]}
       GreenCurve {keys [<at 0 1 RCIM_Cubic>,<at 0.4434 1>,<at 1 1 RCIM_Cubic>]}
       RedCurve {keys [<at 0 0.0635 RCIM_Cubic>,<at 0.4434 0.0724>,<at 1 0.0678 RCIM_Cubic>]}
      }}
     CurveIndex=&UniformRangedFloat(Minimum=0, Maximum=1)
    )) {isSetParameters=true, module=SetVariables_#2}
   &AddVelocity(Velocity=&RandomRangeVector(Minimum=<0 0 50>, Maximum=<0 0 120>))
  ]
  particleUpdateScript [
   &ParticleState()
   /Niagara/Modules/Update/Forces/CurlNoiseForce(
    Noise Frequency=25
    Noise Strength=&RandomRangeFloat(Minimum=250, Maximum=1000)
    Pan Noise=true
    Pan Noise Field=<0 0 1>
   )
   /Niagara/Modules/Update/Forces/V2/CurlNoiseForce(
    Noise Frequency=0.025
    Noise Strength=&RandomRangeFloat(Minimum=180, Maximum=300)
    Pan Noise=true
    Pan Noise Field=<1 1 1>
   ) {module=CurlNoiseForce001}
   /Niagara/Modules/Update/Forces/PointAttractionForce(
    Attraction Radius=1000
    AttractionStrength:NiagaraFloat=&Lerp_Float(A=15, B=75, Alpha=&MakeCustomFloatFromBool(Bool=$User.Attract_To_Jar, True Float=1, False Float=0))
    AttractorPosition:Vector3f=&Lerp_Vector(
     A=&Add_Vector(
      A=/Niagara/DynamicInputs/Helpers/SimulationPosition()
      B=&MakeVector(
       X=&Lerp_Float(
        A=&UniformRangedFloat(Minimum=-250, Maximum=250)
        B=&UniformRangedFloat(Minimum=-800, Maximum=800)
        Alpha=$Emitter.Scalability.DistanceFraction
       )
       Y=&Lerp_Float(
        A=&UniformRangedFloat(Minimum=-250, Maximum=250)
        B=&UniformRangedFloat(Minimum=-800, Maximum=800)
        Alpha=$Emitter.Scalability.DistanceFraction
       )
       Z=&Lerp_Float(
        A=&UniformRangedFloat(Minimum=-190, Maximum=310)
        B=&UniformRangedFloat(Minimum=-300, Maximum=600)
        Alpha=$Emitter.Scalability.DistanceFraction
       )
      )
     )
     B=$User.Jar_Position
     Alpha=&MakeCustomFloatFromBool(Bool=$User.Attract_To_Jar, True Float=1, False Float=0)
    )
    Module.Kill Within Radius=false
    Module.Use Falloff=true
   ) {enabled=false, module=PointAttractionForce002}
   /Niagara/Modules/Update/Forces/V2/PointAttractionForce(
    Attraction Radius=800
    Attraction Strength=&Multiply_Float(
     A=&RandomRangeFloat(Minimum=120, Maximum=200)
     B=&Lerp_Float(
      A=0.2
      B=0.5
      Alpha=&Waveform("Waveform [1]"=::NewEnumerator0, "[1] Frequency"=1, "[1] Amplitude Min/Max"=<0 1>)
     )
    )
    Attractor Position Offset:Vector3f=<0 20 70>
    Kill Within Radius=false
    Module.Use Falloff=true
    UseAttractorOffset=true
   )
   &Drag(Use Linear Drag=true, Use Rotational Drag=false)
   &PointForce(Module.Use Falloff=false, Module.UseLinearForce=false)
   &SolveForcesAndVelocity(Clamp Velocity=false, Limit Acceleration=false)
   &ScaleMeshSize(Scale Factor=&VectorFromFloat(Value=&FloatFromCurve(
      FloatCurve=DataInterface() {propertyValues {Curve {keys [
          {interpMode=RCIM_Cubic, tangentMode=RCTM_User, time=-0.0046, value=1.0113}
          {arriveTangent=-1.0067, interpMode=RCIM_Cubic, leaveTangent=-1.0067, tangentMode=RCTM_User, time=0.9, value=0.55}
          <at 1 0>
         ]}}}
      Scale Curve=2
     )))
   &DynamicMaterialParameters(
    Index 0 Param 1=&UniformRangedFloat(Minimum=4, Maximum=5)
    Index 0 Param 2=&FloatFromCurve(FloatCurve=DataInterface() {propertyValues {Curve {keys [
         {interpMode=RCIM_Cubic, tangentMode=RCTM_User, time=0, value=0}
         {arriveTangent=0.002, interpMode=RCIM_Cubic, leaveTangent=0.002, tangentMode=RCTM_User, time=0.1827, value=1}
         {arriveTangent=-1.2195, interpMode=RCIM_Cubic, leaveTangent=-1.2195, tangentMode=RCTM_User, time=0.9833, value=0.6448}
         <at 1.0011 0.002>
        ]}}})
    Module.Param0WriteEnabled=false
    Module.Param1WriteEnabled=true
    Module.Param2WriteEnabled=false
    Module.Param3WriteEnabled=false
   ) {module=DynamicMaterialParameters001}
   &CameraOffset(Camera Offset Amount=20)
   &ScaleColor(
    Scale RGB=&MakeVectorFromLinearColorRGB(Color=&ColorFromCurve(ColorCurve=DataInterface() {propertyValues {
        AlphaCurve {keys [<at 0 1 RCIM_Cubic>]}
        BlueCurve {keys [
          <at 0.0076 0>
          {arriveTangent=60.2978, interpMode=RCIM_Cubic, leaveTangent=60.2978, time=0.0149, value=1}
          <at 0.0242 1>
          <at 0.0294 0>
          <at 0.2395 0>
          <at 0.281 1>
          <at 0.2871 1>
          <at 0.2905 0>
          <at 0.4693 0>
          <at 0.4933 1>
          <at 0.5053 1>
          <at 0.5172 0>
         ]}
        GreenCurve {keys [
          <at 0.0076 0>
          {arriveTangent=60.2978, interpMode=RCIM_Cubic, leaveTangent=60.2978, time=0.0149, value=1}
          <at 0.0242 1>
          <at 0.0294 0>
          <at 0.2395 0>
          <at 0.281 1>
          <at 0.2871 1>
          <at 0.2905 0>
          <at 0.4693 0>
          <at 0.4933 1>
          <at 0.5053 1>
          <at 0.5172 0>
         ]}
        RedCurve {keys [
          <at 0.0076 0>
          {arriveTangent=60.2978, interpMode=RCIM_Cubic, leaveTangent=60.2978, time=0.0149, value=1}
          <at 0.0242 1>
          <at 0.0294 0>
          <at 0.2395 0>
          <at 0.281 1>
          <at 0.2871 1>
          <at 0.2905 0>
          <at 0.4693 0>
          <at 0.4933 1>
          <at 0.5053 1>
          <at 0.5172 0>
         ]}
       }}))
    ScaleA=true
    ScaleRGB=true
   )
   /SetVariables(Particles.Sprite_Color:LinearColor=&MultiplyLinearColor(
     Linear Color=$Particles.Color
     Scale Factor=&MakeVector4_FromLinearColor(LinearColor=$Particles.Random_Color)
    )) {isSetParameters=true, module=SetVariables_#3}
   &ScaleSpriteSize(Scale Factor=&Vector2DFromFloat(Value=&Lerp_Float(A=1, B=3, Alpha=$Emitter.Scalability.DistanceFraction)))
   &GenerateLocationEvent(Event Send Rate=2, Module.Use Event Send Rate=true) {enabled=false, module=GenerateLocationEvent001}
  ]
  rendererClasses [&NiagaraMeshRendererProperties, &NiagaraSpriteRendererProperties]
  renderers [
   {rendererClass=&NiagaraMeshRendererProperties, rendererIndex=0}
   {rendererClass=&NiagaraSpriteRendererProperties, rendererIndex=1}
  ]
 }
 FN_Smoke {
  emitterUpdateScript [
   &EmitterState(Inactive Response=::NewEnumerator1, Life Cycle Mode=::NewEnumerator1, Loop Duration=1, UseLoopDelay=false) {module=EmitterState001}
   &SpawnRateFromMultipleSpawnGroups_2(Global Spawnrate=5)
  ]
  particleSpawnScript [
   &FN_SetSpawnGroup(Number of Spawn Groups=5)
   &InitializeParticle(
    Color=<rgba 0.651 0.651 0.651 1>
    Lifetime=&UniformRangedFloat(Minimum=0.9, Maximum=1.1)
    Module.Write Color=true
    Module.Write Lifetime=true
    Module.Write Mass=false
    Module.Write Material Random=true
    Module.Write Position=false
    Module.Write Scale=false
    Module.Write SpriteRotation=true
    Module.Write SpriteSize=true
    Position=/Niagara/Functions/Localspace/SimulationPosition()
    Sprite Rotation=&UniformRangedFloat(Minimum=-180, Maximum=180)
    Sprite Size=&Vector2DFromFloat(Value=&Multiply_Float(
      A=&RandomRangeFloat(Minimum=20, Maximum=60)
      B=&MakeCustomFloatFromBool(Bool=&RandomBool(), True Float=-1, False Float=1)
     ))
   )
   &CameraOffset(Camera Offset Amount=&RandomRangeFloat(Minimum=-30, Maximum=30)) {module=CameraOffset001}
   &SkeletalMeshLocation(
    Bone / Socket Index=$Particles.SpawnGroup
    Bone Sampling Mode=::NewEnumerator4
    Inherit Velocity=false
    "Inherit Velocity (Scale)"=0
    MaskSpawnGroup=false
    Sampled Position Offset Coordinate Space:ENiagaraMeshTransforms=::NewEnumerator2
    Skeletal Mesh=$User.SkeletalMesh
   )
   /Niagara/Modules/Spawn/Location/V2/SphereLocation(
    MaskSpawnGroup=false
    Offset=&FN_FrontEndVector3Modifier(Gameplay=<0 0 -10>, Front End=<0 0 20>)
    Override Seed=false
    Sphere Radius=5
    Surface Only=false
    UseNonUniformScale=false
    UseOffset=true
    UseSphereOrientation=false
   )
   &AddVelocity(Velocity=&FN_FrontEndVector3Modifier(Gameplay=<0 0 55>, Front End=<0 -100 80>))
  ]
  particleUpdateScript [
   &SolveForcesAndVelocity(Clamp Velocity=false, Limit Acceleration=false)
   &ParticleState()
   /SetVariables(Particles.SpriteRotation:NiagaraFloat=&Add_Float(A=$Particles.Initial.SpriteRotation, B=&RandomRangeFloat(Minimum=10, Maximum=20))) {isSetParameters=true}
   &ScaleSpriteSize(Scale Factor=&Vector2DFromFloat(Value=&FloatFromCurve(FloatCurve=DataInterface() {propertyValues {Curve {keys [
          {arriveTangent=17.2984, interpMode=RCIM_Cubic, leaveTangent=17.2984, tangentMode=RCTM_User, time=0, value=0.1}
          {arriveTangent=1.5, interpMode=RCIM_Cubic, leaveTangent=1.5, time=0.09, value=1}
          <at 0.6 1>
          <at 1 0.5>
         ]}}})))
   &ScaleColor(
    Linear Color Curve=DataInterface() {propertyValues {
      AlphaCurve {keys [<at 0 0>,<at 0.0853 1>,<at 0.2513 1>,<at 0.4849 0.6>,<at 1 0>]}
      BlueCurve {keys [<at 0.0323 0.246>,<at 0.5097 0.7188>,<at 1 0.3073>]}
      GreenCurve {keys [<at 0.0323 0.3646>,<at 0.5097 0.1241>,<at 1 0.0153>]}
      RedCurve {keys [<at 0.0323 0.0161>,<at 0.5097 0.3135>,<at 1 0.0871>]}
     }}
    Scale Mode=::NewEnumerator2
   ) {module=ScaleColor001}
   &ScaleColor(Scale Alpha=&FN_FrontEndFloatModifier(Gameplay=1, Front End=0.25), ScaleA=true, ScaleRGB=true)
  ]
  rendererClasses [&NiagaraSpriteRendererProperties]
  renderers [{rendererClass=&NiagaraSpriteRendererProperties}]
 }
 Fireflies {
  emitterUpdateScript [
   &EmitterState(
    Enable Distance Culling=true
    Inactive Response=::NewEnumerator1
    Life Cycle Mode=::NewEnumerator1
    Loop Delay=&UniformRangedFloat(Minimum=0.5, Maximum=2.5)
    Loop Duration=&UniformRangedFloat(Minimum=0.6, Maximum=1.8)
    MaxDistance=8000
    Module.bUseMaxDistance=true
    Module.bUseMinDistance=true
    Recalculate Duration Each Loop=true
    Scalability Mode=::NewEnumerator1
    UseLoopDelay=false
   )
   &SpawnBurst_Instantaneous(Spawn Count=3, Use Spawn Probability=false, UseLoopCountLimit=false)
   &SpawnRate(SpawnRate=&Lerp_Float(A=4.5, B=8, Alpha=$User.ParaGlideLeanAlpha), Use Spawn Probability=false)
  ]
  particleSpawnScript [
   &InitializeParticle(
    Lifetime=&UniformRangedFloat(Minimum=1, Maximum=2.5)
    Mesh Scale=&Multiply_VectorByFloat(Vector=<1 2 1>, Float=&UniformRangedFloat(Minimum=0.16, Maximum=0.4))
    Module.Write Color=true
    Module.Write Lifetime=true
    Module.Write Mass=true
    Module.Write Material Random=true
    Module.Write Position=false
    Module.Write Scale=true
    Module.Write SpriteRotation=false
    Module.Write SpriteSize=true
    Position=&SimulationPosition()
    Sprite Size=<30 50>
   )
   &DynamicMaterialParameters(
    Index 0 Param 1=10
    Index 0 Param 2=1
    Index 1 Param 1=&UniformRangedFloat(Minimum=0, Maximum=30)
    Index 1 Param 2=&UniformRangedFloat(Minimum=0, Maximum=1)
    Index 1 Param 3=&UniformRangedFloat(Minimum=0, Maximum=1)
    Index 1 Param 4=&UniformRangedFloat(Minimum=0, Maximum=1)
    Module.Param0WriteEnabled=true
    Module.Param0WriteEnabled1=true
    Module.Param1WriteEnabled=true
    Module.Param1WriteEnabled1=false
    Module.Param2WriteEnabled=true
    Module.Param2WriteEnabled1=false
    Module.Param3WriteEnabled=true
    Module.Param3WriteEnabled1=false
    Write Parameter Index 1=true
   )
   /Niagara/Modules/Spawn/Location/SphereLocation(Hemisphere Z=true, MaskSpawnGroup=false, Module.Override Randomness=false, Offset=<0 0 60>, Sphere Radius=40, Surface Only=false)
   /SetVariables(Particles.Random_Color:LinearColor=&ColorFromCurve(
     ColorCurve=DataInterface() {propertyValues {
       AlphaCurve {keys [<at 0 0.13 RCIM_Cubic>]}
       BlueCurve {keys [<at 0 0.5987 RCIM_Cubic>,<at 0.4434 0.6204>,<at 1 0.7167 RCIM_Cubic>]}
       GreenCurve {keys [<at 0 1 RCIM_Cubic>,<at 0.4434 1>,<at 1 1 RCIM_Cubic>]}
       RedCurve {keys [<at 0 0.0635 RCIM_Cubic>,<at 0.4434 0.0724>,<at 1 0.0678 RCIM_Cubic>]}
      }}
     CurveIndex=&UniformRangedFloat(Minimum=0, Maximum=1)
    )) {isSetParameters=true, module=SetVariables_#2}
   &AddVelocity(
    Module.Use Falloff Distance=false
    Override Seed=false
    Velocity Mode=::NewEnumerator1
    Velocity Speed=&RandomRangeFloat(Minimum=60, Maximum=200)
   ) {module=AddVelocity001}
   &AddVelocity(Velocity=&RandomRangeVector(Minimum=<0 0 50>, Maximum=<0 0 120>))
  ]
  particleUpdateScript [
   &ParticleState()
   /Niagara/Modules/Update/Forces/CurlNoiseForce(
    Noise Frequency=25
    Noise Strength=&RandomRangeFloat(Minimum=250, Maximum=1000)
    Pan Noise=true
    Pan Noise Field=<0 0 1>
   )
   /Niagara/Modules/Update/Forces/V2/CurlNoiseForce(
    Noise Frequency=0.025
    Noise Strength=&RandomRangeFloat(Minimum=180, Maximum=300)
    Pan Noise=true
    Pan Noise Field=<1 1 1>
   ) {module=CurlNoiseForce001}
   /Niagara/Modules/Update/Forces/PointAttractionForce(
    Attraction Radius=1000
    AttractionStrength:NiagaraFloat=&Lerp_Float(A=15, B=75, Alpha=&MakeCustomFloatFromBool(Bool=$User.Attract_To_Jar, True Float=1, False Float=0))
    AttractorPosition:Vector3f=&Lerp_Vector(
     A=&Add_Vector(
      A=/Niagara/DynamicInputs/Helpers/SimulationPosition()
      B=&MakeVector(
       X=&Lerp_Float(
        A=&UniformRangedFloat(Minimum=-250, Maximum=250)
        B=&UniformRangedFloat(Minimum=-800, Maximum=800)
        Alpha=$Emitter.Scalability.DistanceFraction
       )
       Y=&Lerp_Float(
        A=&UniformRangedFloat(Minimum=-250, Maximum=250)
        B=&UniformRangedFloat(Minimum=-800, Maximum=800)
        Alpha=$Emitter.Scalability.DistanceFraction
       )
       Z=&Lerp_Float(
        A=&UniformRangedFloat(Minimum=-190, Maximum=310)
        B=&UniformRangedFloat(Minimum=-300, Maximum=600)
        Alpha=$Emitter.Scalability.DistanceFraction
       )
      )
     )
     B=$User.Jar_Position
     Alpha=&MakeCustomFloatFromBool(Bool=$User.Attract_To_Jar, True Float=1, False Float=0)
    )
    Module.Kill Within Radius=false
    Module.Use Falloff=true
   ) {enabled=false, module=PointAttractionForce002}
   /Niagara/Modules/Update/Forces/V2/PointAttractionForce(
    Attraction Radius=800
    Attraction Strength=&Multiply_Float(
     A=&RandomRangeFloat(Minimum=120, Maximum=200)
     B=&Lerp_Float(
      A=0.2
      B=0.5
      Alpha=&Waveform("Waveform [1]"=::NewEnumerator0, "[1] Frequency"=1, "[1] Amplitude Min/Max"=<0 1>)
     )
    )
    Attractor Position Offset:Vector3f=<0 20 70>
    Kill Within Radius=false
    Module.Use Falloff=true
    UseAttractorOffset=true
   )
   &Drag(Use Linear Drag=true, Use Rotational Drag=false)
   &PointForce(Module.Use Falloff=false, Module.UseLinearForce=false)
   &SolveForcesAndVelocity(Clamp Velocity=false, Limit Acceleration=false)
   &ScaleMeshSize(Scale Factor=&VectorFromFloat(Value=&FloatFromCurve(
      FloatCurve=DataInterface() {propertyValues {Curve {keys [
          {interpMode=RCIM_Cubic, tangentMode=RCTM_User, time=0, value=0.05}
          <at 0.3 1.05>
          <at 0.6 1>
          {arriveTangent=-1.0067, interpMode=RCIM_Cubic, leaveTangent=-1.0067, tangentMode=RCTM_User, time=0.9, value=0.55}
          <at 1 0>
         ]}}}
      Scale Curve=2
     )))
   &DynamicMaterialParameters(
    Index 0 Param 1=&UniformRangedFloat(Minimum=4, Maximum=5)
    Index 0 Param 2=&FloatFromCurve(FloatCurve=DataInterface() {propertyValues {Curve {keys [
         {interpMode=RCIM_Cubic, tangentMode=RCTM_User, time=0, value=0}
         {arriveTangent=0.002, interpMode=RCIM_Cubic, leaveTangent=0.002, tangentMode=RCTM_User, time=0.1827, value=1}
         {arriveTangent=-1.2195, interpMode=RCIM_Cubic, leaveTangent=-1.2195, tangentMode=RCTM_User, time=0.9833, value=0.6448}
         <at 1.0011 0.002>
        ]}}})
    Module.Param0WriteEnabled=false
    Module.Param1WriteEnabled=true
    Module.Param2WriteEnabled=false
    Module.Param3WriteEnabled=false
   ) {module=DynamicMaterialParameters001}
   &CameraOffset(Camera Offset Amount=20)
   &ScaleColor(
    Scale RGB=&MakeVectorFromLinearColorRGB(Color=&ColorFromCurve(ColorCurve=DataInterface() {propertyValues {
        AlphaCurve {keys [<at 0 1 RCIM_Cubic>]}
        BlueCurve {keys [
          <at 0.0048 0>
          <at 0.1748 1>
          {arriveTangent=-5.4949, interpMode=RCIM_Cubic, leaveTangent=-5.4949, time=0.2275, value=0.2135}
          <at 0.3568 0>
          <at 0.4765 0.0677>
          <at 0.5484 1>
          <at 0.7088 0.2344>
          <at 0.9004 0>
         ]}
        GreenCurve {keys [
          <at 0.0048 0>
          <at 0.1748 1>
          {arriveTangent=-5.4949, interpMode=RCIM_Cubic, leaveTangent=-5.4949, time=0.2275, value=0.2135}
          <at 0.3568 0>
          <at 0.4765 0.0677>
          <at 0.5484 1>
          <at 0.7088 0.2344>
          <at 0.9004 0>
         ]}
        RedCurve {keys [
          <at 0.0048 0>
          <at 0.1748 1>
          {arriveTangent=-5.4949, interpMode=RCIM_Cubic, leaveTangent=-5.4949, time=0.2275, value=0.2135}
          <at 0.3568 0>
          <at 0.4765 0.0677>
          <at 0.5484 1>
          <at 0.7088 0.2344>
          <at 0.9004 0>
         ]}
       }}))
    ScaleA=true
    ScaleRGB=true
   )
   /SetVariables(Particles.Sprite_Color:LinearColor=&MultiplyLinearColor(
     Linear Color=$Particles.Color
     Scale Factor=&MakeVector4_FromLinearColor(LinearColor=$Particles.Random_Color)
    )) {isSetParameters=true, module=SetVariables_#3}
   &ScaleSpriteSize(Scale Factor=&Vector2DFromFloat(Value=&FloatFromCurve(FloatCurve=DataInterface() {propertyValues {Curve {keys [
          {interpMode=RCIM_Cubic, leaveTangent=7.0042, tangentMode=RCTM_Break, time=0, value=0}
          <at 0.3 1>
          {interpMode=RCIM_Cubic, leaveTangent=-4.9722, tangentMode=RCTM_Break, time=0.7, value=1}
          <at 1 0>
         ]}}})))
   &GenerateLocationEvent(Event Send Rate=2, Module.Use Event Send Rate=true) {enabled=false, module=GenerateLocationEvent001}
  ]
  rendererClasses [&NiagaraMeshRendererProperties, &NiagaraSpriteRendererProperties]
  renderers [
   {rendererClass=&NiagaraMeshRendererProperties, rendererIndex=0}
   {rendererClass=&NiagaraSpriteRendererProperties, rendererIndex=1}
  ]
 }
 Hand_FeetSmoke {
  emitterUpdateScript [
   &EmitterState()
   &SpawnRate(SpawnRate=20, Use Spawn Probability=false)
   &SpawnBurst_Instantaneous(Spawn Count=4, Spawn Probability=HlslExpression() {hlslExpression="Emitter.LoopCount == 0"}, Use Spawn Probability=true) {enabled=false}
  ]
  particleSpawnScript [
   &FN_SetSpawnGroup()
   /Niagara/Modules/Spawn/Initialization/V2/InitializeParticle(
    Color Maximum=<rgba 0.5906 0.4814 0.7604 0.8>
    Color Minimum=<rgba 0.2339 0.6302 0.4547 0.7>
    Color Mode=::NewEnumerator2
    Lifetime Max=1
    Lifetime Min=0.5
    Lifetime Mode=::NewEnumerator1
    Sprite Rotation Mode=::NewEnumerator2
    Sprite Size Mode=::NewEnumerator3
    Uniform Sprite Size=&RandomRangeFloat(Minimum=25, Maximum=45)
    UsePositionOffset=false
    Write Material Random=true
   ) {module=InitializeParticle001}
   &CameraOffset(Camera Offset Amount=10)
  ]
  particleUpdateScript [
   &ParticleState()
   &SkeletalMeshLocation(
    Apply Sampled Orientation to Mesh Particles=true
    Bone / Socket Index=&SpawnGroup()
    Bone Sampling Mode=::NewEnumerator4
    Inherit Velocity=false
    Interpolate Mesh Sampling=false
    MaskSpawnGroup=false
    Skeletal Mesh=$User.SkeletalMesh
   )
   &ScaleSpriteSize(Uniform Curve Sprite Scale=DataInterface() {propertyValues {Curve {keys [
        {arriveTangent=1.9798, interpMode=RCIM_Cubic, leaveTangent=1.9798, tangentMode=RCTM_User, time=0, value=0.2}
        {interpMode=RCIM_Cubic, tangentMode=RCTM_User, time=1, value=1}
       ]}}})
   &ScaleColor(
    Linear Color Curve=DataInterface() {propertyValues {
      AlphaCurve {keys [<at 0.0067 0>,<at 0.0801 1>,<at 0.1891 1>,<at 1 0>]}
      BlueCurve {keys [<at 0 1>,<at 0.2691 1.1072>,<at 0.5938 0.599>]}
      GreenCurve {keys [<at 0 1>,<at 0.2691 2>,<at 0.5938 0.0124>]}
      RedCurve {keys [<at 0 1>,<at 0.2691 0.1444>,<at 0.5938 0.2997>]}
     }}
    Scale Mode=::NewEnumerator2
   )
  ]
  rendererClasses [&NiagaraSpriteRendererProperties]
  renderers [{rendererClass=&NiagaraSpriteRendererProperties}]
 }
 RibbonDark {
  emitterUpdateScript [
   &EmitterState()
   &FN_SpawnRibbon_v2(Number of Ribbons=2, Spawn Mode=::NewEnumerator0, SpawnRate=64)
  ]
  particleSpawnScript [
   &FN_SetSpawnGroup()
   &FN_InitializeRibbon_v3(
    Color:LinearColor=<rgba 0.3122 0.0819 0.9062 0.8>
    Enable Skeletal Position=true
    Lifetime:NiagaraFloat=&Lerp_Float(A=0.4, B=2.1, Alpha=$User.ParaGlideLeanAlpha)
    Ribbon Width=30
    Sampled Position Coordinate Space:ENiagaraCoordinateSpace=::Local
    Sampled Rotation Coordinate Space:ENiagaraCoordinateSpace=::Local
    Sampled Velocity Coordinate Space:ENiagaraCoordinateSpace=::Local
   ) {module=FN_InitializeRibbon_v3001}
   &AddVelocity(Velocity=&FN_FrontEndVector3Modifier(
     Gameplay=&Lerp_Vector(A=<0 0 300>, B=<0 -200 700>, Alpha=$User.ParaGlideLeanAlpha)
     Front End=&Lerp_Vector(A=<0 -150 300>, B=<1 -400 850>, Alpha=0.5)
    ))
   &SkeletalMeshLocation(Bone / Socket Index=$Particles.SpawnGroup, Bone Sampling Mode=::NewEnumerator4, MaskSpawnGroup=false, Skeletal Mesh=$User.SkeletalMesh)
  ]
  particleUpdateScript [
   &SolveForcesAndVelocity(Clamp Velocity=false, Limit Acceleration=false)
   &FN_UpdateRibbon(
    Enable Skeletal Position=true
    Initial Color:LinearColor=$Particles.Initial.Color
    Initial RibbonWidth:NiagaraFloat=$Particles.Initial.RibbonWidth
   )
   &ScaleRibbonWidth(Ribbon Width Scale=&FloatFromCurve(FloatCurve=DataInterface() {propertyValues {Curve {keys [<at 0 1>,<at 1 3>]}}})) {enabled=false}
   &ScaleColor(
    Linear Color Curve=DataInterface() {propertyValues {
      AlphaCurve {keys [<at 0 0>,<at 0.0489 0.8>,<at 0.1713 1>,<at 0.4537 1>,<at 0.8896 0.2>]}
      BlueCurve {keys [
        {arriveTangent=-0.0469, interpMode=RCIM_Cubic, time=0.1312, value=0.1651}
        {interpMode=RCIM_Cubic, leaveTangent=-0.7216, time=0.9661, value=0.063}
       ]}
      GreenCurve {keys [
        {arriveTangent=-0.1963, interpMode=RCIM_Cubic, time=0.1312, value=0.0356}
        {interpMode=RCIM_Cubic, leaveTangent=-0.6443, time=0.9661, value=0.0086}
       ]}
      RedCurve {keys [
        {arriveTangent=-0.0166, interpMode=RCIM_Cubic, time=0.1312, value=0.0666}
        {arriveTangent=-0.3741, interpMode=RCIM_Cubic, leaveTangent=-0.3741, tangentMode=RCTM_User, time=0.9661, value=0.0212}
       ]}
     }}
    Scale Mode=::NewEnumerator2
   ) {module=ScaleColor001}
  ]
  rendererClasses [&NiagaraRibbonRendererProperties]
  renderers [{rendererClass=&NiagaraRibbonRendererProperties}]
 }
 RibbonDark001 {
  emitterUpdateScript [
   &EmitterState()
   &FN_SpawnRibbon_v2(Number of Ribbons=2, Spawn Mode=::NewEnumerator0, SpawnRate=64)
  ]
  particleSpawnScript [
   &FN_SetSpawnGroup()
   &FN_InitializeRibbon_v3(
    Boneflow Offset=2
    Color:LinearColor=<rgba 0.3122 0.0819 0.9062 0.8>
    Enable Skeletal Position=true
    Lifetime:NiagaraFloat=&Lerp_Float(A=0.3, B=1.5, Alpha=$User.ParaGlideLeanAlpha)
    Ribbon Width=25
    Sampled Position Coordinate Space:ENiagaraCoordinateSpace=::Local
    Sampled Rotation Coordinate Space:ENiagaraCoordinateSpace=::Local
    Sampled Velocity Coordinate Space:ENiagaraCoordinateSpace=::Local
   ) {module=FN_InitializeRibbon_v3001}
   &AddVelocity(Velocity=&FN_FrontEndVector3Modifier(Gameplay=&Lerp_Vector(A=<0 0 300>, B=<0 -80 350>, Alpha=$User.ParaGlideLeanAlpha), Front End=<0 -120 320>))
   &SkeletalMeshLocation(Bone / Socket Index=$Particles.SpawnGroup, Bone Sampling Mode=::NewEnumerator4, MaskSpawnGroup=false, Skeletal Mesh=$User.SkeletalMesh)
  ]
  particleUpdateScript [
   &SolveForcesAndVelocity(Clamp Velocity=false, Limit Acceleration=false)
   &FN_UpdateRibbon(
    Enable Skeletal Position=true
    Initial Color:LinearColor=$Particles.Initial.Color
    Initial RibbonWidth:NiagaraFloat=$Particles.Initial.RibbonWidth
   )
   &ScaleColor(
    Linear Color Curve=DataInterface() {propertyValues {
      AlphaCurve {keys [<at 0 0>,<at 0.0367 1>,<at 0.3459 1>,<at 0.8896 0.1>]}
      BlueCurve {keys [
        {arriveTangent=0.0324, interpMode=RCIM_Cubic, time=0.0578, value=0.1651}
        {interpMode=RCIM_Cubic, leaveTangent=-0.7216, time=0.9661, value=0.063}
       ]}
      GreenCurve {keys [
        {arriveTangent=-0.0451, interpMode=RCIM_Cubic, time=0.0578, value=0.0356}
        {interpMode=RCIM_Cubic, leaveTangent=-0.6443, time=0.9661, value=0.0086}
       ]}
      RedCurve {keys [
        {arriveTangent=0.0173, interpMode=RCIM_Cubic, time=0.0578, value=0.0666}
        {arriveTangent=-0.3741, interpMode=RCIM_Cubic, leaveTangent=-0.3741, tangentMode=RCTM_User, time=0.9661, value=0.0212}
       ]}
     }}
    Scale Mode=::NewEnumerator2
   ) {module=ScaleColor001}
  ]
  rendererClasses [&NiagaraRibbonRendererProperties]
  renderers [{rendererClass=&NiagaraRibbonRendererProperties}]
 }
 RibbonLighter {
  emitterUpdateScript [
   &EmitterState()
   &FN_SpawnRibbon_v2(Number of Ribbons=2, Spawn Mode=::NewEnumerator0, SpawnRate=64)
  ]
  particleSpawnScript [
   &FN_SetSpawnGroup()
   &FN_InitializeRibbon_v3(
    Color:LinearColor=<rgba 0.2941 0.9062 0.6352 0.65>
    Enable Skeletal Position=true
    Lifetime:NiagaraFloat=&Lerp_Float(A=0.65, B=2, Alpha=$User.ParaGlideLeanAlpha)
    Ribbon Width=18
    Sampled Position Coordinate Space:ENiagaraCoordinateSpace=::Local
    Sampled Rotation Coordinate Space:ENiagaraCoordinateSpace=::Local
    Sampled Velocity Coordinate Space:ENiagaraCoordinateSpace=::Local
   ) {module=FN_InitializeRibbon_v3001}
   &AddVelocity(Velocity=&FN_FrontEndVector3Modifier(
     Gameplay=&Lerp_Vector(A=<0 0 400>, B=<0 -100 700>, Alpha=$User.ParaGlideLeanAlpha)
     Front End=&Lerp_Vector(A=<0 -150 300>, B=<1 -300 650>, Alpha=$User.ParaGlideLeanAlpha)
    ))
   &SkeletalMeshLocation(Bone / Socket Index=$Particles.SpawnGroup, Bone Sampling Mode=::NewEnumerator4, MaskSpawnGroup=false, Skeletal Mesh=$User.SkeletalMesh)
  ]
  particleUpdateScript [
   &SolveForcesAndVelocity(Clamp Velocity=false, Limit Acceleration=false)
   &FN_UpdateRibbon(
    Enable Skeletal Position=true
    Initial Color:LinearColor=$Particles.Initial.Color
    Initial RibbonWidth:NiagaraFloat=$Particles.Initial.RibbonWidth
   )
   &ScaleColor(
    Linear Color Curve=DataInterface() {propertyValues {
      AlphaCurve {keys [<at 0 0>,<at 0.0595 1>,<at 0.7316 0.5>,<at 1 0.1>]}
      BlueCurve {keys [
        <at 0 1.3927>
        <at 0.2136 0.3241>
        {arriveTangent=-0.4065, interpMode=RCIM_Cubic, leaveTangent=-0.4065, time=0.407, value=0.5208}
        {interpMode=RCIM_Cubic, leaveTangent=-0.7216, time=0.9661, value=1}
       ]}
      GreenCurve {keys [
        <at 0 2>
        <at 0.2136 0.5417>
        {arriveTangent=-1.0351, interpMode=RCIM_Cubic, leaveTangent=-1.0351, time=0.407, value=0.0436}
        {interpMode=RCIM_Cubic, leaveTangent=-0.6443, time=0.9661, value=1}
       ]}
      RedCurve {keys [
        <at 0 0.0236>
        <at 0.2136 0.0395>
        {arriveTangent=1.0107, interpMode=RCIM_Cubic, leaveTangent=1.0107, time=0.407, value=0.2146}
        {arriveTangent=-0.3741, interpMode=RCIM_Cubic, leaveTangent=-0.3741, tangentMode=RCTM_User, time=0.9661, value=1}
       ]}
     }}
    Scale Mode=::NewEnumerator2
   ) {module=ScaleColor001}
  ]
  rendererClasses [&NiagaraRibbonRendererProperties]
  renderers [{rendererClass=&NiagaraRibbonRendererProperties}]
 }
 RibbonLighter001 {
  emitterUpdateScript [
   &EmitterState()
   &FN_SpawnRibbon_v2(Number of Ribbons=2, Spawn Mode=::NewEnumerator0, SpawnRate=64)
  ]
  particleSpawnScript [
   &FN_SetSpawnGroup()
   &FN_InitializeRibbon_v3(
    Boneflow Offset=2
    Color:LinearColor=<rgba 0.2941 0.9062 0.6352 0.65>
    Enable Skeletal Position=true
    Lifetime:NiagaraFloat=&Lerp_Float(A=0.4, B=1.5, Alpha=$User.ParaGlideLeanAlpha)
    Ribbon Width=16
    Sampled Position Coordinate Space:ENiagaraCoordinateSpace=::Local
    Sampled Rotation Coordinate Space:ENiagaraCoordinateSpace=::Local
    Sampled Velocity Coordinate Space:ENiagaraCoordinateSpace=::Local
   ) {module=FN_InitializeRibbon_v3001}
   &AddVelocity(Velocity=&FN_FrontEndVector3Modifier(Gameplay=&Lerp_Vector(A=<0 0 300>, B=<0 -80 350>, Alpha=$User.ParaGlideLeanAlpha), Front End=<0 -180 320>))
   &SkeletalMeshLocation(Bone / Socket Index=$Particles.SpawnGroup, Bone Sampling Mode=::NewEnumerator4, MaskSpawnGroup=false, Skeletal Mesh=$User.SkeletalMesh)
  ]
  particleUpdateScript [
   &SolveForcesAndVelocity(Clamp Velocity=false, Limit Acceleration=false)
   &FN_UpdateRibbon(
    Enable Skeletal Position=true
    Initial Color:LinearColor=$Particles.Initial.Color
    Initial RibbonWidth:NiagaraFloat=$Particles.Initial.RibbonWidth
   )
   &ScaleColor(
    Linear Color Curve=DataInterface() {propertyValues {
      AlphaCurve {keys [<at 0 0>,<at 0.0508 1>,<at 0.7316 0.8>,<at 1 0.1>]}
      BlueCurve {keys [
        <at 0.0044 1.5584>
        <at 0.1157 0.2986>
        <at 0.1979 0.2986>
        {arriveTangent=0.8248, interpMode=RCIM_Cubic, leaveTangent=0.8248, time=0.3203, value=0.5208}
        {interpMode=RCIM_Cubic, leaveTangent=-0.7216, time=0.9661, value=1}
       ]}
      GreenCurve {keys [
        <at 0.0044 2>
        <at 0.1157 0.4427>
        <at 0.1979 0.4427>
        {arriveTangent=0.6553, interpMode=RCIM_Cubic, leaveTangent=0.6553, time=0.3203, value=0.0436}
        {interpMode=RCIM_Cubic, leaveTangent=-0.6443, time=0.9661, value=1}
       ]}
      RedCurve {keys [
        <at 0.0044 0.009>
        <at 0.1157 0.0325>
        <at 0.1979 0.0325>
        {arriveTangent=1.1377, interpMode=RCIM_Cubic, leaveTangent=1.1377, time=0.3203, value=0.2146}
        {arriveTangent=-0.3741, interpMode=RCIM_Cubic, leaveTangent=-0.3741, tangentMode=RCTM_User, time=0.9661, value=1}
       ]}
     }}
    Scale Mode=::NewEnumerator2
   ) {module=ScaleColor001}
  ]
  rendererClasses [&NiagaraRibbonRendererProperties]
  renderers [{rendererClass=&NiagaraRibbonRendererProperties}]
 }
}
summary {
 systemName=NS_Contrail_StinkyCloud_v4
 emitters [
  {emitterName=Hand_FeetSmoke, rendererClasses [&NiagaraSpriteRendererProperties]}
  {emitterName=RibbonLighter, rendererClasses [&NiagaraRibbonRendererProperties]}
  {emitterName=RibbonDark, rendererClasses [&NiagaraRibbonRendererProperties]}
  {emitterName=RibbonLighter001, rendererClasses [&NiagaraRibbonRendererProperties]}
  {emitterName=RibbonDark001, rendererClasses [&NiagaraRibbonRendererProperties]}
  {emitterName=FN_Smoke, rendererClasses [&NiagaraSpriteRendererProperties]}
  {emitterName=FN_AttachedParticle_v3, rendererClasses [&NiagaraSpriteRendererProperties]}
  {
   bEnabled=false
   emitterName=FN_Fireflies_Optimized
   rendererClasses [&NiagaraMeshRendererProperties, &NiagaraSpriteRendererProperties]
  }
  {emitterName=Empty, rendererClasses [&NiagaraRibbonRendererProperties]}
  {emitterName=Empty001, rendererClasses [&NiagaraRibbonRendererProperties]}
  {emitterName=Fireflies, rendererClasses [&NiagaraMeshRendererProperties, &NiagaraSpriteRendererProperties]}
  {emitterName=Decal, rendererClasses [&NiagaraSpriteRendererProperties]}
 ]
 userVariables [
  User.Attract_To_Jar {type=NiagaraBool}
  User.bIsFrontEnd {type=NiagaraBool}
  User.bIsFrontEndPreview {type=NiagaraBool}
  User.IsSkydiving {type=NiagaraBool}
  User.Jar_Position {type=Vector3f, description="", defaultValue {struct=Vector3f, value=<0 0 0>}}
  User.ParaGlideLeanAlpha {type=NiagaraFloat, defaultValue=0, description=""}
  User.PlayerSpeed {type=NiagaraFloat, defaultValue=0, description=""}
  User.RibbonIDOffset {type=NiagaraInt32}
  User.SkeletalMesh {
   type=NiagaraDataInterfaceSkeletalMesh
   defaultValue {
    struct=NiagaraExt_VariableValue_DataInterface
    value {dataInterface=&self:NiagaraDataInterfaceSkeletalMesh_0}
   }
  }
  User.SkeletalMesh2 {
   type=NiagaraDataInterfaceSkeletalMesh
   defaultValue {
    struct=NiagaraExt_VariableValue_DataInterface
    value {dataInterface=&self:NiagaraDataInterfaceSkeletalMesh_2}
   }
  }
 ]
}
